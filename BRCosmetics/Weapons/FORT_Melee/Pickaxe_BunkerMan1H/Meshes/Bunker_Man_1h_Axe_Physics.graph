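# Physics asset for the BunkerMan one-handed pickaxe mesh in the BRCosmetics melee weapons.
# A single capsule covers the Melee_TwoHanded_Handle bone, and smaller capsules sit on the left and right dyn_peel bones.
# Each peel bone is constrained to the handle with matching asymmetric swing limits and a tight twist limit, so the peels can swing within a narrow range relative to the handle.
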PhysicsAsset /BRCosmetics/Weapons/FORT_Melee/Pickaxe_BunkerMan1H/Meshes/Bunker_Man_1h_Axe_Physics
bodies {
 Melee_TwoHanded_Handle [{shapeType=Capsule, length=40, radius=5, center=<0 0 13.0699>}]
 dyn_peel_l [{shapeType=Capsule, length=1.6524, radius=1.4532, center=<1 0 0>, rotation=<rot 90 0 0>}]
 dyn_peel_r [{shapeType=Capsule, length=5, radius=1.75, center=<4.4054 0 -1.588>, rotation=<rot 65 0 0>}]
}
constraints [
 {bone1Name=dyn_peel_l, bone2Name=Melee_TwoHanded_Handle, swing1LimitDegrees=20, swing2LimitDegrees=40, twistLimitDegrees=5}
 {bone1Name=dyn_peel_r, bone2Name=Melee_TwoHanded_Handle, swing1LimitDegrees=20, swing2LimitDegrees=40, twistLimitDegrees=5}
]
